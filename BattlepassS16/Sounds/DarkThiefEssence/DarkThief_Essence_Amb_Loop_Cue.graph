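# Looping ambient sound cue for the Dark Thief Essence, routed to the loot ambient loop sound class.
# A mixer layers the essence ambience loop with a quieter life-drain loop that a modulator plays at a fixed lowered pitch.
# An enveloper at the root fades the mix in over roughly half a second.
# Overridden attenuation uses logarithmic falloff with listener focus, and the shared loot ambient concurrency rule is assigned.
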SoundCue /BattlepassS16/Sounds/DarkThiefEssence/DarkThief_Essence_Amb_Loop_Cue
root=Enveloper_0
nodes [
 {
  id=Enveloper_0
  children [Mixer_0]
  properties {pitchCurve {editorCurveData {keys [{time=0}]}}, volumeCurve {editorCurveData {keys [<at 0.0172 0.0116>,<at 0.5444 1>]}}}
 }
 {id=Mixer_0, children [WavePlayer_1,Modulator_0], properties {inputVolume [1,0.6]}}
 {id=Modulator_0, children [WavePlayer_0], properties {pitchMax=0.8, pitchMin=0.8}}
 {id=WavePlayer_0, properties {bLooping=true, soundWaveAssetPtr=&/DarkThief_LifeDrain_Layer_Loop}}
 {id=WavePlayer_1, properties {bLooping=true, soundWaveAssetPtr=&/DarkThief_Essence_Amb_Loop_01}}
]
properties {
 bOverrideAttenuation=true
 maxDistance=1600
 soundClassObject=/Game/Sounds/FortSoundClassesAndModes/Fort_Loot_Ambient_Loop
 attenuationOverrides {
  bEnableListenerFocus=true
  distanceAlgorithm=Logarithmic
  falloffDistance=1300
  nonFocusVolumeAttenuation=0.4762
  stereoSpread=1000
  attenuationShapeExtents {x=300}
 }
 concurrencyOverrides {maxCount=2}
 concurrencySet [/Game/Sounds/Concurrency_Rules/Loot_Ambient_Conc]
}
