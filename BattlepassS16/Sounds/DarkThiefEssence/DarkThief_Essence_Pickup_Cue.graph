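# Sound cue for picking up the DarkThief Essence in Battle Pass S16, played as a loot-collect sound with weapon pickup attenuation.
# A mixer layers the unique essence pickup sound with one of three randomly chosen large held-object pickup sounds.
# A third layer, a reward-screen chest loot upgrade sting, enters after a short fixed delay at reduced mixer volume.
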SoundCue /BattlepassS16/Sounds/DarkThiefEssence/DarkThief_Essence_Pickup_Cue
root=Mixer_0
nodes [
 {id=Delay_0, children [WavePlayer_3], properties {delayMax=0.3, delayMin=0.3}}
 {id=Mixer_0, children [Modulator_2,Modulator_0,Delay_0], properties {inputVolume [1,1,0.6]}}
 {id=Modulator_0, children [Random_0], properties {volumeMax=1, volumeMin=1}}
 {id=Modulator_2, children [WavePlayer_0], properties {volumeMax=1, volumeMin=1}}
 {id=Random_0, children [WavePlayer_1, WavePlayer_2, WavePlayer_10], properties {weights [1,1,1]}}
 {id=WavePlayer_0, properties {soundWaveAssetPtr=&/DarkThief_Essence_Pickup_01}}
 {id=WavePlayer_1, properties {soundWaveAssetPtr=/FNGameplayCues/Sounds/Apollo/HeldObjects/HeldObjects_Large_Pickup_01}}
 {id=WavePlayer_10, properties {soundWaveAssetPtr=/FNGameplayCues/Sounds/Apollo/HeldObjects/HeldObjects_Large_Pickup_03}}
 {id=WavePlayer_2, properties {soundWaveAssetPtr=/FNGameplayCues/Sounds/Apollo/HeldObjects/HeldObjects_Large_Pickup_02}}
 {id=WavePlayer_3, properties {soundWaveAssetPtr=/Game/Sounds/Fort_UI_Sounds/Rewards/Fort_RewardScreen_TeamScore_ChestLootUpgrade}}
]
properties {
 attenuationSettings=&Fort_Weapon_Pickup_Att
 maxDistance=3500
 soundClassObject=&Fort_Loot_Collect
 volumeMultiplier=0.6
}
